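# Looping stereo MetaSound source for the DragonCart flying orb, layering a main close loop with an echo loop.
# The echo layer gets linear distance attenuation before mixing, then the enclosure modulation bus low-passes and lowers the mix as enclosure rises.
# An inverted focus multiplier from azimuth and elevation reduces gain when the orb is in focus.
# The Mobile page keeps only the main loop and focus ducking.
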MetaSoundSource /DragonCartNarrative/Assets/Audio/FlyingOrb/MSS_DragonCart_FlyingOrb_Echo_LP
inputs [
 EchoLoop {type=WaveAsset}
 GainReductionInFocus {type=Float}
 Sound {type=WaveAsset}
 UE.Attenuation.Distance {type=Float}
 UE.Source.OnPlay {type=Trigger}
 UE.Spatialization.Azimuth {type=Float}
 UE.Spatialization.Elevation {type=Float}
 Wave Asset {type=WaveAsset}
]
interfaces [UE.Attenuation:1.0, UE.OutputFormat.Stereo:1.0, UE.Source:1.1, UE.Spatialization:1.0]
outputs [
 "UE.OutputFormat.Stereo.Audio:0" {type=Audio}
 "UE.OutputFormat.Stereo.Audio:1" {type=Audio}
]
pages [
 Default {
  is_build_graph=true
  input_defaults {
   EchoLoop=&/FlyingOrb_Aggro_LP_01
   GainReductionInFocus=0.7
   Sound=&/FlyingOrb_Close_LP
   UE.Attenuation.Distance=0
   UE.Source.OnPlay=null
   UE.Spatialization.Azimuth=0
   UE.Spatialization.Elevation=0
   Wave Asset=null
  }
  nodes [
   n1 = Modulation.GetModulatorValue() {also {defaults {Modulator=&AudioWorldization_Bus_Enclosure}}}
   n2 = MapRange.MapRange:Float(In=@n1.Out) {also {defaults {In Range A=0.75, In Range B=0.9, Out Range A=1, Out Range B=0.2}}}
   n3 = /FortMetasound/Attenuation/MSP_Attenuation_Volume_Linear(Distance=^UE.Attenuation.Distance) {also {defaults {FalloffDistance=32000, InnerRadius=2000}}}
   WavePlayer_1 = UE.Wave Player:Stereo(Wave Asset=^EchoLoop, Play=^UE.Source.OnPlay) {also {defaults {Loop="<MetaSoundBuilderResult.SUCCEEDED: 0>"}, version [1,1]}}
   Multiply_0 = UE.Multiply:Audio by Float(AdditionalOperands=@n3.VolumeMultiplier, PrimaryOperand=@WavePlayer_1.Out Left)
   WavePlayer_0 = UE.Wave Player:Stereo(Wave Asset=^Sound, Play=^UE.Source.OnPlay) {also {defaults {Loop="<MetaSoundBuilderResult.SUCCEEDED: 0>"}, version [1,1]}}
   AudioMixerAudioMixerStereo2_0 = "AudioMixer.Audio Mixer (Stereo, 2)"(
    In 1 L=@Multiply_0.Out
    In 1 R=UE.Multiply:Audio by Float(AdditionalOperands=@n3.VolumeMultiplier, PrimaryOperand=@WavePlayer_1.Out Right).Out
    In 0 L=@WavePlayer_0.Out Left
    In 0 R=@WavePlayer_0.Out Right
   )
   n4 = MapRange.MapRange:Float(In=@n1.Out) {also {defaults {In Range A=0.75, In Range B=0.9, Out Range A=20000, Out Range B=150}}}
   OnePoleLowPassFilter_1 = UE.One-Pole Low Pass Filter:Audio(In=@AudioMixerAudioMixerStereo2_0.Out L, Cutoff Frequency=@n4.Out Value)
   n5 = UE.Subtract:Float(AdditionalOperands=/FortMetasound/Attenuation/MSP_Attenuation_Focus(FocusValue=^GainReductionInFocus, Azimuth=^UE.Spatialization.Azimuth, Elevation=^UE.Spatialization.Elevation) {also {defaults {
       FocusAngleThreshold=20
       FocusAttackInterpSpeed=30
       FocusReleaseInterpSpeed=8
       NonFocusAngleThreshold=45
       NonFocusValue=0
       bEnableFocusInterpolation="<MetaSoundBuilderResult.SUCCEEDED: 0>"
      }}}.FocusMultiplier) {also {defaults {PrimaryOperand=1}}}
   Multiply_1 = UE.Multiply:Audio by Float(
    PrimaryOperand=UE.Multiply:Audio by Float(AdditionalOperands=@n2.Out Value, PrimaryOperand=@OnePoleLowPassFilter_1.Out).Out
    AdditionalOperands=@n5.Out
   )
   OnePoleLowPassFilter_0 = UE.One-Pole Low Pass Filter:Audio(In=@AudioMixerAudioMixerStereo2_0.Out R, Cutoff Frequency=@n4.Out Value)
   Multiply_2 = UE.Multiply:Audio by Float(
    PrimaryOperand=UE.Multiply:Audio by Float(AdditionalOperands=@n2.Out Value, PrimaryOperand=@OnePoleLowPassFilter_0.Out).Out
    AdditionalOperands=@n5.Out
   )
  ]
  links [{dst="@UE.OutputFormat.Stereo.Audio:0", src=Multiply_1.Out}, {dst="@UE.OutputFormat.Stereo.Audio:1", src=Multiply_2.Out}]
 }
 Mobile {
  nodes [
   n1 = UE.Subtract:Float(AdditionalOperands=&MSP_Attenuation_Focus(FocusValue=^GainReductionInFocus, Azimuth=^UE.Spatialization.Azimuth, Elevation=^UE.Spatialization.Elevation) {also {defaults {
       FocusAngleThreshold=20
       FocusAttackInterpSpeed=30
       FocusReleaseInterpSpeed=8
       NonFocusAngleThreshold=45
       NonFocusValue=0
       bEnableFocusInterpolation="<MetaSoundBuilderResult.SUCCEEDED: 0>"
      }}}.FocusMultiplier) {also {defaults {PrimaryOperand=1}}}
   WavePlayer_0 = UE.Wave Player:Stereo(Wave Asset=^Sound, Play=^UE.Source.OnPlay) {also {defaults {Loop="<MetaSoundBuilderResult.SUCCEEDED: 0>"}, version [1,1]}}
   Multiply_0 = UE.Multiply:Audio by Float(AdditionalOperands=@n1.Out, PrimaryOperand=@WavePlayer_0.Out Left)
   Multiply_1 = UE.Multiply:Audio by Float(AdditionalOperands=@n1.Out, PrimaryOperand=@WavePlayer_0.Out Right)
  ]
  links [{dst="@UE.OutputFormat.Stereo.Audio:0", src=Multiply_0.Out}, {dst="@UE.OutputFormat.Stereo.Audio:1", src=Multiply_1.Out}]
 }
]
properties {
 assetImportData=&self:AssetImportData
 attenuationSettings=&/ATT_DragonCart_FlyingOrb_Basic
 bLooping=true
 duration=10000
 maxDistance=182000
 numChannels=2
 outputFormat=Stereo
 sampleRate=48000
 soundClassObject=&Fort_Events_Gameplay
 volume=0.5
 modulationSettings {
  volumeRouting=Union
  volumeModulationDestination {
   bEnableModulation=true
   modulators [/Hera/Props/SkyMango/SkyMango_Fragment_A/Audio/CB_SkyMango_Fragment, &/CB_DragonCart_FlyingOrb]
  }
 }
}
