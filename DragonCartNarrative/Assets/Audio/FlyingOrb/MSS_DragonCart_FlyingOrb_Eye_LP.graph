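# Looping mono MetaSound source for the Dragon Cart flying orb's eye, synthesized from FM sine oscillators with noise-modulated pitch and a filter with wandering cutoff.
# Smoothed absolute azimuth scales carrier pitch and tremolo rate, while elevation offsets modulator frequency, so the sound reacts to its direction from the listener.
# The filtered tone mixes with a tremolo-gated pitch-shifted copy, then passes through feedback delay and flanger.
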MetaSoundSource /DragonCartNarrative/Assets/Audio/FlyingOrb/MSS_DragonCart_FlyingOrb_Eye_LP
inputs [
 Delay Pitch Shift {type=Float}
 FM Amount {type=Float}
 Frequency Carrier {type=Float}
 Frequency Modulator {type=Float}
 Master Pitch Offset {type=Float}
 Tremolo LFO Rate {type=Float}
 UE.Source.OnPlay {type=Trigger}
 UE.Spatialization.Azimuth {type=Float}
 UE.Spatialization.Elevation {type=Float}
]
interfaces [UE.OutputFormat.Mono:1.0, UE.Source:1.1, UE.Spatialization:1.0]
outputs ["UE.OutputFormat.Mono.Audio:0" {type=Audio}]
pages [
 Default {
  is_build_graph=true
  input_defaults {
   Delay Pitch Shift=19.248
   FM Amount=6526.2632
   Frequency Carrier=3744
   Frequency Modulator=40.6
   Master Pitch Offset=8.0435
   Tremolo LFO Rate=3.8041
   UE.Source.OnPlay=null
   UE.Spatialization.Azimuth=0
   UE.Spatialization.Elevation=0
  }
  nodes [
   LfoFrequencyNoise_0 = UE.Lfo Frequency Noise:Audio() {also {defaults {Max Value=3000, Min Value=400, Rate=1.7, Rate Jitter=0.5}}}
   n1 = MapRange.MapRange:Float(In=AbsoluteValue.Abs:Float(Input=^UE.Spatialization.Elevation).Value) {also {defaults {In Range B=50, Out Range B=10}}}
   n2 = UE.Sine:Audio(Frequency=UE.Add:Float(
     PrimaryOperand=UE.Add:Float(PrimaryOperand=^Frequency Modulator, AdditionalOperands=^Master Pitch Offset).Out
     AdditionalOperands=UE.InterpTo:Audio(Target=@n1.Out Value) {also {defaults {Interp Time=1}}}.Value
    ).Out)
   LfoFrequencyNoise_1 = UE.Lfo Frequency Noise:Audio() {also {defaults {Interpolation=0, Max Value=2500, Rate=1.9, Rate Jitter=1.2}}}
   n3 = UE.Add:Float(
    PrimaryOperand=UE.Add:Float(PrimaryOperand=^Frequency Carrier, AdditionalOperands=^Master Pitch Offset).Out
    AdditionalOperands=@LfoFrequencyNoise_1.Out
   )
   n4 = UE.InterpTo:Audio(Target=MapRange.MapRange:Float(In=AbsoluteValue.Abs:Float(Input=^UE.Spatialization.Azimuth).Value) {also {defaults {In Range B=50, Out Range A=3}}}.Out Value) {also {defaults {Interp Time=1}}}
   BiquadFilter_0 = UE.Biquad Filter:Audio(
    Cutoff Frequency=UE.Multiply:Float(PrimaryOperand=@LfoFrequencyNoise_0.Out) {also {defaults {AdditionalOperands=0.5}}}.Out
    In=UE.Sine:Audio(
     Modulation=UE.Multiply:Audio by Float(AdditionalOperands=^FM Amount, PrimaryOperand=@n2.Audio).Out
     Frequency=UE.Multiply:Float(PrimaryOperand=@n3.Out, AdditionalOperands=@n4.Value).Out
    ).Audio
   ) {also {defaults {Bandwidth=3.5, Type=3}}}
   DelayPitchShift_0 = UE.Delay Pitch Shift:Audio(Pitch Shift=^Delay Pitch Shift, In=@BiquadFilter_0.Out)
   n5 = MapRange.MapRange:Float(In=@n4.Value) {also {defaults {In Range A=3, Out Range A=15}}}
   AudioMixerAudioMixerMono2_0 = "AudioMixer.Audio Mixer (Mono, 2)"(
    In 0=@BiquadFilter_0.Out
    In 1=@DelayPitchShift_0.Out
    Gain 1=UE.LFO:Audio(Frequency=UE.Add:Float(AdditionalOperands=^Tremolo LFO Rate, PrimaryOperand=@n5.Out Value).Out) {also {defaults {Min Value=0, Shape=3}}}.Out
   ) {also {defaults {Gain 0=0.5}}}
   Flanger_0 = UE.Flanger(In Audio=UE.Delay:Audio(In=@AudioMixerAudioMixerMono2_0.Out) {also {defaults {Delay Time=0.06, Dry Level=0.5, Feedback=0.6, Wet Level=0.5}}}.Out) {also {defaults {Center Delay=0.15}}}
  ]
  links [{dst="@UE.OutputFormat.Mono.Audio:0", src=Flanger_0.Out Audio}]
 }
 Mobile {links [], nodes []}
]
properties {
 assetImportData=&self:AssetImportData
 attenuationSettings=&/ATT_DragonCart_FlyingOrb_Eye
 bLooping=true
 duration=10000
 maxDistance=5400
 sampleRate=48000
 soundClassObject=&Fort_Events_Gameplay
 volume=0.14
 concurrencyOverrides {maxCount=1}
 modulationSettings {
  volumeRouting=Union
  volumeModulationDestination {bEnableModulation=true, modulators [&/CB_DragonCart_FlyingOrb]}
 }
}
