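# Mono looping MetaSound source for the DragonCart flying orb's shield, with attenuation and control-bus volume modulation.
# A wave player starts at a random point, and a short ramp fades in master volume.
# The default page mixes the dry loop with a sweetener: noise-modulated biquad filter into delay pitch shift, then feedback delay.
# The Mobile page drops that effects chain, keeping only the ramped dry loop.
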MetaSoundSource /DragonCartNarrative/Assets/Audio/FlyingOrb/MSS_DragonCart_FlyingOrb_Shield_LP
inputs [
 Delay Time {type=Audio}
 "Master Volume (Linear)" {type=Float}
 Pitch Shift {type=Float}
 Sound {type=WaveAsset}
 "Sweetener Volume (Linear)" {type=Float}
 UE.Source.OnPlay {type=Trigger}
]
interfaces [UE.OutputFormat.Mono:1.0, UE.Source:1.1]
outputs ["UE.OutputFormat.Mono.Audio:0" {type=Audio}]
pages [
 Default {
  is_build_graph=true
  input_defaults {
   Delay Time=1
   "Master Volume (Linear)"=1
   Pitch Shift=12
   Sound=&/DragonCart_FlyingOrb_Shield_LP
   "Sweetener Volume (Linear)"=0.5
   UE.Source.OnPlay=null
  }
  nodes [
   LfoFrequencyNoise_0 = UE.Lfo Frequency Noise:Audio() {also {defaults {Interpolation=0, Max Value=4000, Min Value=800, Rate=3.8}}}
   GetWaveDuration_0 = "UE.Get Wave Duration: "(Wave=^Sound)
   WavePlayer_0 = UE.Wave Player:Mono(Wave Asset=^Sound, Play=^UE.Source.OnPlay, Start Time=UE.RandomTime(Max=@GetWaveDuration_0.Duration).Value) {also {defaults {Loop="<MetaSoundBuilderResult.SUCCEEDED: 0>"}, version [1,1]}}
   BiquadFilter_0 = UE.Biquad Filter:Audio(Cutoff Frequency=@LfoFrequencyNoise_0.Out, In=@WavePlayer_0.Out Mono) {also {defaults {Bandwidth=0.2, Gain=6, Type=0}}}
   DelayPitchShift_0 = UE.Delay Pitch Shift:Audio(Pitch Shift=^Pitch Shift, In=@BiquadFilter_0.Out)
   n1 = HarmonixNodes.Ramper(Target=^"Master Volume (Linear)", Snap To Start=@WavePlayer_0.On Play) {also {defaults {Ramp Time=0.1}, version [0,1]}}
   AudioMixerAudioMixerMono2_0 = "AudioMixer.Audio Mixer (Mono, 2)"(
    In 1=UE.Delay:Audio(In=@DelayPitchShift_0.Out) {also {defaults {Delay Time=0.1, Feedback=0.5}}}.Out
    Gain 0=@n1.Value
    Gain 1=UE.Multiply:Float(AdditionalOperands=^"Sweetener Volume (Linear)", PrimaryOperand=@n1.Value).Out
    In 0=@WavePlayer_0.Out Mono
   )
  ]
  links [{dst="@UE.OutputFormat.Mono.Audio:0", src=AudioMixerAudioMixerMono2_0.Out}]
 }
 Mobile {
  nodes [
   GetWaveDuration_0 = "UE.Get Wave Duration: "(Wave=^Sound)
   WavePlayer_0 = UE.Wave Player:Mono(Wave Asset=^Sound, Play=^UE.Source.OnPlay, Start Time=UE.RandomTime(Max=@GetWaveDuration_0.Duration).Value) {also {defaults {Loop="<MetaSoundBuilderResult.SUCCEEDED: 0>"}, version [1,1]}}
   AudioMixerAudioMixerMono2_0 = "AudioMixer.Audio Mixer (Mono, 2)"(
    Gain 0=HarmonixNodes.Ramper(Target=^"Master Volume (Linear)", Snap To Start=@WavePlayer_0.On Play) {also {defaults {Ramp Time=0.1, Start=0}, version [0,1]}}.Value
    In 0=@WavePlayer_0.Out Mono
   ) {also {defaults {Gain 1=0.5}}}
  ]
  links [{dst="@UE.OutputFormat.Mono.Audio:0", src=AudioMixerAudioMixerMono2_0.Out}]
 }
]
properties {
 assetImportData=&self:AssetImportData
 attenuationSettings=&/ATT_DragonCart_FlyingOrb_Shield
 bLooping=true
 duration=10000
 maxDistance=4767.8164
 sampleRate=48000
 soundClassObject=&Fort_Events_Gameplay
 volume=0.35
 modulationSettings {
  volumeRouting=Union
  volumeModulationDestination {bEnableModulation=true, modulators [&/CB_DragonCart_FlyingOrb]}
 }
}
