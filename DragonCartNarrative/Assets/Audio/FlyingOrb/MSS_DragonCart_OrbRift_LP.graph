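# Looping stereo MetaSound source for the DragonCart flying orb rift, driven by listener distance.
# The Default page loops from a random start, with curve-based volume and a high-pass cutoff that rises as smoothed distance shrinks.
# A gate on that distance mixes in random close one-shots on entering and exiting its threshold.
# The Mobile page plays only the loop, scaled by inverted focus attenuation.
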MetaSoundSource /DragonCartNarrative/Assets/Audio/FlyingOrb/MSS_DragonCart_OrbRift_LP
inputs [
 Looping Sound {type=WaveAsset}
 OneShots {type=WaveAsset:Array}
 UE.Attenuation.Distance {type=Float}
 UE.Source.OnPlay {type=Trigger}
]
interfaces [UE.Attenuation:1.0, UE.OutputFormat.Stereo:1.0, UE.Source:1.1]
outputs [
 "UE.OutputFormat.Stereo.Audio:0" {type=Audio}
 "UE.OutputFormat.Stereo.Audio:1" {type=Audio}
]
pages [
 Default {
  is_build_graph=true
  input_defaults {
   Looping Sound=&/DragonCart_OrbRift_LP
   UE.Attenuation.Distance=643.3214
   UE.Source.OnPlay=null
   OneShots [
    &/DragonCart_OrbRift_Close_01
    &/DragonCart_OrbRift_Close_02
    &/DragonCart_OrbRift_Close_03
   ]
  }
  nodes [
   n1 = UE.InterpTo:Audio(Target=^UE.Attenuation.Distance)
   n2 = MapRange.MapRange:Float(In=@n1.Value) {also {defaults {In Range A=2000, In Range B=5000, Out Range A=2500, Out Range B=10}}}
   n3 = &MSP_Attenuation_Volume_Curve(Distance=^UE.Attenuation.Distance) {also {defaults {AttenuationCurves=&/WTB_DragonCart_OrbRift_Distance, FalloffDistance=9000, InnerRadius=1000}}}
   GetWaveDuration_0 = "UE.Get Wave Duration: "(Wave=^Looping Sound)
   n4 = UE.RandomTime(Next=^UE.Source.OnPlay, Max=@GetWaveDuration_0.Duration)
   WavePlayer_0 = UE.Wave Player:Stereo(Wave Asset=^Looping Sound, Play=@n4.On Next, Start Time=@n4.Value) {also {defaults {Loop="<MetaSoundBuilderResult.SUCCEEDED: 0>"}, version [1,1]}}
   Multiply_0 = UE.Multiply:Audio by Float(AdditionalOperands=@n3.VolumeMultiplier, PrimaryOperand=@WavePlayer_0.Out Left)
   OnePoleHighPassFilter_0 = UE.One-Pole High Pass Filter:Audio(Cutoff Frequency=@n2.Out Value, In=@Multiply_0.Out)
   Multiply_1 = UE.Multiply:Audio by Float(AdditionalOperands=@n3.VolumeMultiplier, PrimaryOperand=@WavePlayer_0.Out Right)
   OnePoleHighPassFilter_1 = UE.One-Pole High Pass Filter:Audio(Cutoff Frequency=@n2.Out Value, In=@Multiply_1.Out)
   n5 = &MSP_OneshotGate(Value=@n1.Value) {also {defaults {EnterCondition=4, ExitCondition=3, OneshotThreshold=4000}}}
   TriggerAnyTriggerAny2_0 = "TriggerAny.Trigger Any (2)"(In 0=@n5.OnGateEnter, In 1=@n5.OnGateExit)
   ArrayRandomGet_0 = Array.Random Get:WaveAsset:Array(In Array=^OneShots, Next=@TriggerAnyTriggerAny2_0.Out)
   WavePlayer_1 = UE.Wave Player:Stereo(Play=@ArrayRandomGet_0.On Next, Wave Asset=@ArrayRandomGet_0.Value) {also {version [1,1]}}
   AudioMixerAudioMixerStereo2_0 = "AudioMixer.Audio Mixer (Stereo, 2)"(In 0 L=@OnePoleHighPassFilter_0.Out, In 0 R=@OnePoleHighPassFilter_1.Out, In 1 L=@WavePlayer_1.Out Left, In 1 R=@WavePlayer_1.Out Right) {also {defaults {Gain 1=0.6}}}
  ]
  links [
   {dst="@UE.OutputFormat.Stereo.Audio:0", src=AudioMixerAudioMixerStereo2_0.Out L}
   {dst="@UE.OutputFormat.Stereo.Audio:1", src=AudioMixerAudioMixerStereo2_0.Out R}
  ]
 }
 Mobile {
  nodes [
   n1 = UE.Subtract:Float(AdditionalOperands=/FortMetasound/Attenuation/MSP_Attenuation_Focus() {also {defaults {
       FocusAngleThreshold=20
       FocusAttackInterpSpeed=30
       FocusReleaseInterpSpeed=8
       FocusValue=0.7
       NonFocusAngleThreshold=45
       NonFocusValue=0
       bEnableFocusInterpolation="<MetaSoundBuilderResult.SUCCEEDED: 0>"
      }}}.FocusMultiplier) {also {defaults {PrimaryOperand=1}}}
   WavePlayer_0 = UE.Wave Player:Stereo(Wave Asset=^Looping Sound, Play=^UE.Source.OnPlay) {also {defaults {Loop="<MetaSoundBuilderResult.SUCCEEDED: 0>"}, version [1,1]}}
   Multiply_0 = UE.Multiply:Audio by Float(AdditionalOperands=@n1.Out, PrimaryOperand=@WavePlayer_0.Out Left)
   Multiply_1 = UE.Multiply:Audio by Float(AdditionalOperands=@n1.Out, PrimaryOperand=@WavePlayer_0.Out Right)
  ]
  links [{dst="@UE.OutputFormat.Stereo.Audio:0", src=Multiply_0.Out}, {dst="@UE.OutputFormat.Stereo.Audio:1", src=Multiply_1.Out}]
 }
]
properties {
 assetImportData=&self:AssetImportData
 attenuationSettings=&/ATT_DragonCart_FlyingOrb_Rift
 bLooping=true
 duration=10000
 maxDistance=29882.7617
 numChannels=2
 outputFormat=Stereo
 sampleRate=48000
 soundClassObject=&Fort_Events_Gameplay
 volume=0.5
 modulationSettings {
  volumeRouting=Union
  volumeModulationDestination {
   bEnableModulation=true
   modulators [/Hera/Props/SkyMango/SkyMango_Fragment_A/Audio/CB_SkyMango_Fragment, &/CB_DragonCart_FlyingOrb]
  }
 }
}
